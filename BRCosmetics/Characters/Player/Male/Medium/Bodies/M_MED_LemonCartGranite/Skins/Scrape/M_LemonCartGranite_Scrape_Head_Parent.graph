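# Parent head material for the LemonCartGranite Scrape skin that paints a view-dependent animated mouth into base color between the CharacterParent_2023 PreFX and PostFX functions.
# Camera direction from the snout, compared against snout right and forward vectors, selects front or side mouth textures, mirrors them and shifts their scaled UV1 placement.
# An index parameter picks a 4x4 flipbook cell.
# Debug switches show UVs, cycle frames or override the side.
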Material /BRCosmetics/Characters/Player/Male/Medium/Bodies/M_MED_LemonCartGranite/Skins/Scrape/M_LemonCartGranite_Scrape_Head_Parent
graph {
 properties {
  bHasPixelAnimation=true
  bUseMaterialAttributes=true
  bUsedWithMeshDeformer=true
  bUsedWithMorphTargets=true
  bUsedWithSkeletalMesh=true
  pixelDepthOffsetMode=PDOM_Legacy
  previewMesh=/BRCosmetics/Characters/Player/Male/Medium/Heads/M_MED_LemonCartGranite_Scrape_Head/Meshes/M_MED_LemonCartGranite_Scrape_Head
  shadingModel=MSM_FromMaterialExpression
  usedShadingModels=Default Lit | Subsurface | Clear Coat | Subsurface Profile | Cloth
 }
 nodes [
  n1 = Multiply(A=ObjectPositionWS() {originType=CameraRelative}) {constB=-1}
  n2 = MaterialFunctionCall(Float3=Transform(None=Normalize(VectorInput=@n1)) {transformSourceType=TRANSFORMSOURCE_World, transformType=TRANSFORM_Local}) {materialFunction=&BreakOutFloat3Components}
  LinearInterpolate(A=ConstantBiasScale(None=Sign(None=@n2.R)), B=ConstantBiasScale(None=@n2.R), Alpha=Saturate(None=Sign(None=@n2.G)))
  n3 = MaterialFunctionCall() {materialFunction=&MF_CharacterParent_2023_PreFX}
  n4 = GetMaterialAttributes(None=@n3.Result) {attributeGetTypes [#1]}
  n5 = NamedRerouteUsage()
  n6 = LinearInterpolate(A=@n4.Base Color, B=TextureSample(UVs=NamedRerouteUsage()) {texture=/Niagara/DefaultAssets/T_NiagaraUVDebug}.RGB, Alpha=@n5) {constA=0.1}
  n7 = Round(None=NamedRerouteUsage())
  n8 = Constant() {r=4}
  n9 = Divide(A=Constant2Vector() {g=1, r=1}, B=@n8) {desc=Scale}
  n10 = ComponentMask(None=@n9) {desc="1 / ## Horz Images", r=true}
  n11 = AppendVector(
   A=Fmod(A=@n7, B=ComponentMask(None=@n8) {desc="## Horz Images", r=true}) {desc=IndexA_Horz}
   B=Floor(None=Multiply(A=@n7, B=@n10)) {desc=IndexA_Vert}
  ) {desc=IndexA_Combined}
  n12 = MaterialFunctionCall(Float2=NamedRerouteUsage()) {materialFunction=&BreakOutFloat2Components}
  n13 = Saturate(None=NamedRerouteUsage())
  n14 = NamedRerouteUsage()
  n15 = AppendVector(A=LinearInterpolate(A=OneMinus(None=@n12.R), B=@n12.R, Alpha=Min(A=@n13, B=@n14)), B=@n12.G)
  n16 = NamedRerouteDeclaration(None=Multiply(A=Add(A=@n11, B=@n15), B=@n9) {desc=UV_A}) {name=Mouth_FlipbookUV, nodeColor=<rgba 1 0 0.4453 1>}
  n17 = TextureSampleParameter2D(UVs=@n16) {
   parameterName=Mouth_Front
   texture=/BRCosmetics/Characters/Player/Male/Medium/Bodies/M_MED_LemonCartGranite/Skins/Scrape/Textures/T_LemonCartGranite_Scrape_Mouth_Front_D
  }
  n18 = TextureSampleParameter2D(UVs=@n16) {
   parameterName=Mouth_Side
   texture=/BRCosmetics/Characters/Player/Male/Medium/Bodies/M_MED_LemonCartGranite/Skins/Scrape/Textures/T_LemonCartGranite_Scrape_Mouth_Side_D
  }
  n19 = NamedRerouteUsage()
  n20 = LinearInterpolate(
   A=@n4.Base Color
   B=LinearInterpolate(A=@n17.RGB, B=@n18.RGB, Alpha=NamedRerouteUsage())
   Alpha=Multiply(A=@n5, B=LinearInterpolate(A=@n17.A, B=@n18.A, Alpha=@n19))
  ) {constA=0.1}
  MaterialFunctionCall(
   Material=SetMaterialAttributes!(MaterialAttributes=@n4.MaterialAttributes, Base Color=StaticSwitchParameter(True=@n6, False=@n20) {parameterName=Debug Mouth UVs}) {attributeSetTypes [#1]}
   M=@n3.M
   Pre Skinned Local Position=@n3.M
   Distance Blend=@n3.M
   Fresnel=@n3.M
  ) {materialFunction=&MF_CharacterParent_2023_PostFX}
  Max(A=@n13, B=@n14)
  n21 = Time() {bOverride_Period=true, period=16}
  n22 = ScalarParameter() {parameterName=flipbook_face_mouth_index, sliderMax=15}
  NamedRerouteDeclaration(None=StaticSwitchParameter(True=Floor(None=@n21), False=Floor(None=@n22)) {parameterName=Debug_CycleMouthIndex}) {name=Mouth Index, nodeColor=<rgba 0 1 0.7656 1>}
  n23 = ScalarParameter() {defaultValue=0.375, parameterName=Mouth U MaxOffset}
  n24 = TransformPosition(None=VectorParameter() {parameterName=SnoutPosition}.RGB) {transformType=TRANSFORMPOSSOURCE_TranslatedWorld}
  n25 = Transform(None=Normalize(VectorInput=Multiply(A=@n24) {constB=-1})) {transformSourceType=TRANSFORMSOURCE_World, transformType=TRANSFORM_Local}
  n26 = DotProduct(A=VectorParameter() {parameterName=SnoutRightVector, defaultValue=<rgba 0 1 0 0>}.RGB, B=@n25)
  n27 = Sign(None=@n26)
  n28 = Sign(None=DotProduct(A=VectorParameter() {parameterName=SnoutForwardVector, defaultValue=<rgba -1 0 0 0>}.RGB, B=@n25))
  n29 = Subtract(A=LinearInterpolate(A=Saturate(None=@n27), B=ConstantBiasScale(None=@n26), Alpha=Saturate(None=@n28))) {constB=0.5}
  n30 = ScalarParameter() {parameterName=MouthDebug}
  n31 = Step(X=ScalarParameter() {parameterName=UseMouthDebug}) {constY=0.5}
  n32 = NamedRerouteDeclaration(None=LinearInterpolate(A=Sign(None=@n29), B=@n30, Alpha=@n31)) {name=LeftRightSign, nodeColor=<rgba 1 0 0.1172 1>}
  n33 = Saturate(None=Subtract(A=Abs(None=@n29), B=ScalarParameter() {defaultValue=0.1, parameterName=Mouth Frontal Angle}))
  n34 = Abs(None=@n30)
  n35 = Constant() {r=1}
  n36 = Constant()
  n37 = NamedRerouteDeclaration(None=LinearInterpolate(A=Sign(None=@n33), B=If(A=@n34, "A > B"=@n35, "A == B"=@n35, "A < B"=@n36) {constB=0.5}, Alpha=@n31)) {name=FrontSideMask, nodeColor=<rgba 1 0 0.6562 1>}
  n38 = LinearInterpolate(A=Multiply(A=@n23) {constB=-1}, B=@n23, Alpha=ConstantBiasScale(None=Multiply(A=@n32, B=@n37)))
  n39 = LinearInterpolate(A=ScalarParameter() {parameterName=Mouth V Offset Front}, B=ScalarParameter() {parameterName=Mouth V Offset Side}, Alpha=@n37)
  n40 = MaterialFunctionCall(
   UVs=Subtract(A=TextureCoordinate() {coordinateIndex=1}, B=AppendVector(A=@n38, B=@n39))
   Texture Scale=ScalarParameter() {defaultValue=0.25, parameterName=Mouth Scale}
  ) {materialFunction=&ScaleUVsByCenter}
  NamedRerouteDeclaration(None=@n40.UVs) {name=Mouth Repositioned UV, nodeColor=<rgba 1 0.4453 0 1>}
  NamedRerouteDeclaration(None=@n40.0-1 mask) {name=Mouth0-1Mask, nodeColor=<rgba 0.125 0 1 1>}
 ]
}
